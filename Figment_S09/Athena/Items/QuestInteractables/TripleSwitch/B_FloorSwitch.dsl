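; Floor pressure-plate switch, apparently for the Triple Switch quest interactable, that stays pressed while player pawns stand on its pad.
; PadTrigger overlaps set the replicated SwitchState, releasing only once no FortPlayerPawn remains.
; OnRep_SwitchState plays press or release sounds and timelines, which cancel each other and drive the button's scale and height via UpdateButton.
; Finished timelines call BroadcastPressed or BroadcastReleased, both empty here.
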(event EventBeginPlay
  (Variables|Default|SetSwitchState "NewEnumerator1"))

(event Custom|Switch%20Pressed
  (Audio|PlaySoundatLocation "/Figment_S09/Athena/Sounds/Quests/Athena_ShootingGallery_PressurePlate_ON__Cue.Athena_ShootingGallery_PressurePlate_ON__Cue" (Transformation|GetActorLocation))
  (Utilities|FlowControl|Sequence
    (:then_0
      (bind (_direction _buttonscale _buttonz) (Timeline|Timeline_SwitchRelease Stop)
        (:Update
          (CallFunction|UpdateButton _buttonscale _buttonz))
        (:Finished
          (CallFunction|BroadcastReleased))))
    (:then_1
      (bind (_direction_2 _buttonscale_2 _buttonz_2) (Timeline|Timeline_SwitchPress PlayFromStart)
        (:Update
          (CallFunction|UpdateButton _buttonscale_2 _buttonz_2))
        (:Finished
          (CallFunction|BroadcastPressed))))))

(event Custom|SwitchReleased
  (Audio|PlaySoundatLocation "/Figment_S09/Athena/Sounds/Quests/Athena_ShootingGallery_PressurePlate_OFF_Cue.Athena_ShootingGallery_PressurePlate_OFF_Cue" (Transformation|GetActorLocation))
  (Utilities|FlowControl|Sequence
    (:then_0
      (Timeline|Timeline_SwitchPress Stop))
    (:then_1
      (Timeline|Timeline_SwitchRelease PlayFromStart))))

(event OnComponentBeginOverlap(PadTrigger)
  (params (OverlappedComponent :type Primitive%20Component%20Object%20Reference) (OtherActor :type Actor%20Object%20Reference) (OtherComp :type Primitive%20Component%20Object%20Reference) (OtherBodyIndex :type Integer) (bFromSweep :type Boolean) (SweepResult :type Hit%20Result%20Structure%20%28by%20ref%29))
  (Utilities|Casting|CastToFortPlayerPawn OtherActor)
  (Variables|Default|SetSwitchState "NewEnumerator2"))

(event OnComponentEndOverlap(PadTrigger)
  (params (OverlappedComponent :type Primitive%20Component%20Object%20Reference) (OtherActor :type Actor%20Object%20Reference) (OtherComp :type Primitive%20Component%20Object%20Reference) (OtherBodyIndex :type Integer))
  (if (not (Utilities|IsArrayNonEmpty (Collision|GetOverlappingActors (Variables|Default|GetPadTrigger) "/Script/FortniteGame.FortPlayerPawn" :!class /Script/Engine.PrimitiveComponent)))
    (Variables|Default|SetSwitchState "NewEnumerator1")))

(fn BroadcastPressed)

(fn BroadcastReleased)

(fn OnRep_SwitchState
  (switch Utilities|FlowControl|Switch|SwitchonEnum_SwitchState (Variables|Default|GetSwitchState)
    (:NewEnumerator1
      (CallFunction|SwitchReleased))
    (:NewEnumerator2
      (CallFunction|SwitchPressed))))

(fn UpdateButton
  (params (ButtonScale :type Float%20%28double-precision%29) (ButtonZ :type Float%20%28double-precision%29))
  (bind _floor_button (Variables|Default|GetFloor_Button))
  (Transformation|SetRelativeScale3D _floor_button (Math|Vector|MakeVector 1.0 1.0 ButtonScale))
  (Transformation|SetRelativeLocation _floor_button (Math|Vector|MakeVector 0.0 0.0 ButtonZ) false false))

(fn ConstructionScript)
